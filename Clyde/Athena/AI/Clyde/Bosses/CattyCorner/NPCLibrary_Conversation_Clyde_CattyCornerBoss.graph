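# Conversation database for the Catty Corner boss NPC in Clyde, giving the player a short dialogue tree under the Boss1 entry tag.
# The entry leads to a "Mrow?" speech node behind an "Ask about..." choice, which branches to a question about its mission, a "familiar face" question, and a Back option.
# Each answer is a flavor speech line followed by a ReturnToStart node that restarts the conversation.
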ConversationDatabase /Clyde/Athena/AI/Clyde/Bosses/CattyCorner/NPCLibrary_Conversation_Clyde_CattyCornerBoss
entry_points [{tag=Conversation.Entry.Clyde.Boss.CattyCorner.Boss1, destinations [#1]}]
nodes [
 ConversationEntryPointNode_0 {
  id=#1
  class=ConversationEntryPointNode
  outputs [#2]
  properties {entryTag {tagName=Conversation.Entry.Clyde.Boss.CattyCorner.Boss1}}
 }
 Speech_C_0 {
  id=#2
  class=Speech_C
  outputs [#3, #4, #5]
  properties {nodeName="", softTaskIcon=&T_UI_NPC_question, generalConfig {defaultMessage=Mrow?}}
  sub_nodes [ConversationChoiceNode_5 {class=ConversationChoiceNode, properties {defaultChoiceDisplayText=Ask about...}}]
 }
 Speech_C_4 {
  id=#5
  class=Speech_C
  outputs [#6]
  properties {nodeName="", softTaskIcon=&T_UI_NPC_question, generalConfig {defaultMessage=Were we in the same litter?}}
  sub_nodes [ConversationChoiceNode_3 {class=ConversationChoiceNode, properties {defaultChoiceDisplayText=You look familiar.}}]
 }
 ReturnToStart_C_0 {
  id=#6
  class=ReturnToStart_C
  properties {softTaskIcon=&T_UI_NPC_Back}
  sub_nodes [ConversationChoiceNode_1 {class=ConversationChoiceNode, properties {defaultChoiceDisplayText=BACK}}]
 }
 Back_C_3 {
  id=#4
  class=Back_C
  sub_nodes [ConversationChoiceNode_23 {class=ConversationChoiceNode, properties {defaultChoiceDisplayText=BACK}}]
 }
 Speech_C_3 {
  id=#3
  class=Speech_C
  outputs [#7]
  properties {
   nodeName=""
   softTaskIcon=&T_UI_NPC_question
   generalConfig {defaultMessage="Keep it movin' kitty cat. I'm on a mission to defend this place from intruders!"}
  }
  sub_nodes [ConversationChoiceNode_0 {class=ConversationChoiceNode, properties {defaultChoiceDisplayText=What are you up to?}}]
 }
 ReturnToStart_C_1 {
  id=#7
  class=ReturnToStart_C
  properties {softTaskIcon=&T_UI_NPC_Back}
  sub_nodes [ConversationChoiceNode_2 {class=ConversationChoiceNode, properties {defaultChoiceDisplayText=BACK}}]
 }
]
